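# Sound cue for the NyxGlass grapple glove break, in the gadgets and items sound class.
# A mixer layers two random picks: one of five NyxGlass break waves and one of four borrowed BlackMonday cable-tighten waves.
# A root modulator adds slight random volume variation with a boost above unity.
# Grappling gadget attenuation spatializes it, and per-owner concurrency keeps at most two playing, stopping the oldest.
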SoundCue /DryBox/Items/NyxGlass/Audio/FromGrappleGlove/NyxGlass_GG_Break_Cue
root=Modulator_2
nodes [
 {id=Mixer_0, children [Random_2,Random_0], properties {inputVolume [1,1]}}
 {id=Modulator_2, children [Mixer_0], properties {volumeMax=1.3, volumeMin=1.2}}
 {id=Random_0, children [WavePlayer_0,WavePlayer_1,WavePlayer_4,WavePlayer_5,WavePlayer_6], properties {weights [1,1,1,1,1]}}
 {id=Random_2, children [WavePlayer_2, WavePlayer_3, WavePlayer_11, WavePlayer_12], properties {weights [1,1,1,1]}}
 {id=WavePlayer_0, properties {soundWaveAssetPtr=&/NyxGlass_GG_Break_01}}
 {id=WavePlayer_1, properties {soundWaveAssetPtr=&/NyxGlass_GG_Break_02}}
 {id=WavePlayer_11, properties {soundWaveAssetPtr=/Game/Sounds/Fort_Weapons/BlackMonday/Grape/BM_Grape_CableTightens_03}}
 {id=WavePlayer_12, properties {soundWaveAssetPtr=/Game/Sounds/Fort_Weapons/BlackMonday/Grape/BM_Grape_CableTightens_04}}
 {id=WavePlayer_2, properties {soundWaveAssetPtr=/Game/Sounds/Fort_Weapons/BlackMonday/Grape/BM_Grape_CableTightens_01}}
 {id=WavePlayer_3, properties {soundWaveAssetPtr=/Game/Sounds/Fort_Weapons/BlackMonday/Grape/BM_Grape_CableTightens_02}}
 {id=WavePlayer_4, properties {soundWaveAssetPtr=&/NyxGlass_GG_Break_03}}
 {id=WavePlayer_5, properties {soundWaveAssetPtr=&/NyxGlass_GG_Break_04}}
 {id=WavePlayer_6, properties {soundWaveAssetPtr=&/NyxGlass_GG_Break_05}}
]
properties {
 attenuationSettings=/Game/Sounds/Attenuation/Fort_GadgetGrappling_Att
 maxDistance=8500
 soundClassObject=&Fort_Gadgets_Items
 concurrencySet [/Game/Sounds/Concurrency_Rules/TwoMaximum_Owner_StopOldest]
}
