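# Sound cue for the NyxGlass destroy sound, filed under the item's grapple glove audio.
# A Random node picks one of two destroy waves with equal weight and feeds a Modulator that slightly varies pitch.
# The cue raises base pitch and volume, uses the loot collect sound class, and overrides attenuation and concurrency, with concurrency set to prevent new instances at the limit.
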SoundCue /DryBox/Items/NyxGlass/Audio/FromGrappleGlove/NyxGlass_GG_Destroy_Cue
root=Modulator_0
nodes [
 {id=Modulator_0, children [Random_0], properties {pitchMax=1.1, pitchMin=0.9}}
 {id=Random_0, children [WavePlayer_1,WavePlayer_2], properties {weights [1,1]}}
 {id=WavePlayer_1, properties {soundWaveAssetPtr=&/NyxGlass_GG_Destroy_01}}
 {id=WavePlayer_2, properties {soundWaveAssetPtr=&/NyxGlass_GG_Destroy_02}}
]
properties {
 attenuationSettings=/Game/Sounds/Attenuation/Fort_Prop_Search_Att
 bOverrideAttenuation=true
 bOverrideConcurrency=true
 maxDistance=2800
 pitchMultiplier=1.1
 soundClassObject=&Fort_Loot_Collect
 volumeMultiplier=1.23
 attenuationOverrides {falloffDistance=2000, attenuationShapeExtents {x=800}}
 concurrencyOverrides {resolutionRule=PreventNew}
}
