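# Sound cue for the looping in-use swing sound of the NyxGlass item, using audio from a GrappleGlove folder.
# A random node picks one of three equally weighted, looping horizontal swing waves.
# A continuous modulator at the root drives pitch and volume from a Velocity parameter, so the loop follows movement speed.
# It plays in the gadgets and items sound class with third-person grappling movement attenuation.
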SoundCue /DryBox/Items/NyxGlass/Audio/FromGrappleGlove/NyxGlass_GG_InUse_Loop_Cue
root=ModulatorContinuous_0
nodes [
 {
  id=ModulatorContinuous_0
  children [Random_0]
  properties {
   pitchModulationParams {default=0, maxOutput=1.5, minOutput=1, parameterName=Velocity}
   volumeModulationParams {default=0, minOutput=0.5, parameterName=Velocity}
  }
 }
 {id=Random_0, children [WavePlayer_6,WavePlayer_7,WavePlayer_8], properties {weights [1,1,1]}}
 {
  id=WavePlayer_6
  properties {bLooping=true, soundWaveAssetPtr=&/NyxGlass_GG_SwingLoop_Horizontal_01}
 }
 {
  id=WavePlayer_7
  properties {bLooping=true, soundWaveAssetPtr=&/NyxGlass_GG_SwingLoop_Horizontal_02}
 }
 {
  id=WavePlayer_8
  properties {bLooping=true, soundWaveAssetPtr=&/NyxGlass_GG_SwingLoop_Horizontal_03}
 }
]
properties {
 attenuationSettings=/Game/Sounds/Attenuation/Fort_Grappling_Movement_3P_Linear_Att
 maxDistance=11500
 soundClassObject=&Fort_Gadgets_Items
 volumeMultiplier=0.7
}
